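# Niagara cloud effect for the SniffRust emote, drawing tinted sprite puffs around the owner over about eight seconds.
# Only Cloud001 is enabled: spawn groups supply an opening puff, a lingering cloud and a fast closing burst, with a late point force and rising drag acting on the cloud.
# Age curves drive material parameters, color and opacity, adjusted for time-of-day and front-end; sprites follow owner movement and flatten by camera angle.
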NiagaraSystem /CosmeticCompanions/FX/Builtins/SniffRust/NS_Emote_SniffRust_Clouds
dependencies {
 usedDataInterfaces [
  &NiagaraDataInterfaceCurve
  &NiagaraDataInterfaceArrayFloat
  &NiagaraDataInterfaceCamera
  &NiagaraDataInterfaceArrayFloat3
  &NiagaraDataInterfaceColorCurve
 ]
 usedDynamicInputs [&Multiply_Vector]
 usedModules [
  &SystemState
  &Timeline
  /BRCosmetics/Characters/Player/Female/Medium/Bodies/F_MED_PiperShelf_Pearl/FX/TimeOfDayGlobal
 ]
}
emitters {
 Cloud {
  bEnabled=false
  emitterSpawnScript [
   /SetVariables(Emitter.FakeLightActive=true) {isSetParameters=true, module=SetVariables_#1}
   /SetVariables(Emitter.FrontEndShadowLow:NiagaraFloat=&FN_FrontEndFloatModifier(Gameplay=0.7, Front End=1.5)) {isSetParameters=true, module=SetVariables_#2}
  ]
  emitterUpdateScript [
   &EmitterState(Life Cycle Mode=::NewEnumerator1, Loop Behavior=::NewEnumerator1, Loop Duration=7.25, UseLoopDelay=false)
   &SpawnBurst_Instantaneous(Spawn Count=8, Spawn Group=1, Use Spawn Probability=false, UseLoopCountLimit=false)
   &SpawnRate(
    SpawnRate=&Multiply_Float(
     A=10
     B=&FloatFromCurve(FloatCurve=DataInterface() {propertyValues {Curve {keys [<at 0 0 RCIM_Constant>,<at 0.2 1 RCIM_Constant>]}}}, CurveIndex=$Emitter.Age)
    )
    Use Spawn Probability=false
   )
   &SpawnBurst_Instantaneous(Spawn Count=7, Spawn Group=2, Spawn Time=7.2, Use Spawn Probability=false, UseLoopCountLimit=false) {module=SpawnBurst_Instantaneous001}
  ]
  particleSpawnScript [
   &InitializeParticle(
    Color=<rgba 0.6 0.6 0.6 1>
    Lifetime=&MaskFloat(
     Mask Source=0
     Mask Parameter=?
     Mask=2
     Masked Float=&RandomRangeFloat(Minimum=0.4, Maximum=0.6)
     Passthrough Float=&RandomRangeFloat(Minimum=3, Maximum=4)
    )
    Sprite Size Mode=::NewEnumerator1
    Sprite UV Mode=::NewEnumerator3
    Uniform Sprite Size Max=60
    Uniform Sprite Size Min=45
    UsePositionOffset=false
    Write Material Random=true
   )
   &ShapeLocation(
    Cylinder Height=0
    Cylinder Radius=&MaskFloatBySpawnGroup(
     Masked Float=50
     Passthrough Float=&MaskFloatBySpawnGroup(Masked Float=10, Passthrough Float=20, Spawn Group Mask=1)
    )
    MaskSpawnGroup=false
    Offset=&SelectVectorFromArray(
     Direct Array Index=&SpawnGroup()
     Vector Selection Array=DataInterface() {propertyValues {FloatData [<0 0 0>, <0 0 35>, <0 0 0>]}}
    )
    Offset Mode=::NewEnumerator0
    Override Seed=false
    Shape Primitive=::NewEnumerator1
    Surface Only=true
    Surface Only Band Thickness:NiagaraFloat=&SelectFloatFromArray(
     Direct Array Index=&SpawnGroup()
     Float Selection Array=DataInterface() {propertyValues {FloatData [0.8,0,0.5]}}
    )
    Use Endcaps In Surface Only Mode=false
    UseOwnerScale=true
   )
   &AddVelocity(
    Module.Use Falloff Distance=false
    Origin Offset=&Lerp_Vector(
     A=<0 0 35>
     B=<0 0 0>
     Alpha=&FloatFromCurve(
      FloatCurve=DataInterface() {propertyValues {Curve {defaultValue=1, keys [{tangentMode=RCTM_User, time=0, value=0}, {tangentMode=RCTM_User, time=0.5, value=1}]}}}
      CurveIndex=$Emitter.Age
     )
    )
    Override Seed=false
    Velocity Mode=::NewEnumerator1
    Velocity Speed=&Multiply_Float(
     A=&RandomRangeFloat(Minimum=20, Maximum=50)
     B=&FloatFromCurve(
      FloatCurve=DataInterface() {propertyValues {Curve {defaultValue=1, keys [{tangentMode=RCTM_User, time=0, value=1}, {tangentMode=RCTM_User, time=0.5, value=0}]}}}
      CurveIndex=$Emitter.Age
     )
    )
   ) {enabled=false}
   &CameraOffset(Camera Offset Amount=50)
   &ApplyInitialForces()
   &AddVelocity(
    Module.Use Falloff Distance=false
    Origin Offset=<0 0 55>
    Override Seed=false
    Velocity Mode=::NewEnumerator1
    Velocity Speed=&Multiply_Float(
     A=&SelectFloatFromArray(Direct Array Index=&SpawnGroup(), Float Selection Array=DataInterface() {propertyValues {FloatData [0,60,0]}})
     B=&RandomRangeFloat(Minimum=0.8, Maximum=1.1)
    )
   ) {module=AddVelocity001}
   &AddVelocity(
    Module.Use Falloff Distance=false
    Override Seed=false
    Velocity Mode=::NewEnumerator1
    Velocity Speed=&Multiply_Float(
     A=&RandomRangeFloat(Minimum=600, Maximum=900)
     B=&FloatFromCurve(
      FloatCurve=DataInterface() {propertyValues {Curve {keys [<at 6.8 0 RCIM_Constant>,<at 7.1 1 RCIM_Constant>]}}}
      CurveIndex=$Emitter.Age
      Scale Curve=&MaskFloatBySpawnGroup(Masked Float=0, Passthrough Float=1)
     )
    )
   ) {module=AddVelocity002}
  ]
  particleUpdateScript [
   &InheritSourceMovement(
    Applied Position Scale=&Multiply_VectorByFloat(
     Vector=<1 1 1>
     Float=&Lerp_Float(
      A=&FloatFromCurve(FloatCurve=DataInterface() {propertyValues {Curve {keys [{arriveTangent=-1.0658, interpMode=RCIM_Cubic, leaveTangent=-1.0658, tangentMode=RCTM_User, time=0.1, value=1}, <at 1 0.4 RCIM_Cubic>]}}})
      B=&FloatFromCurve(FloatCurve=DataInterface() {propertyValues {Curve {keys [{arriveTangent=-0.0965, interpMode=RCIM_Cubic, leaveTangent=-0.0965, tangentMode=RCTM_User, time=0.3, value=1}, <at 1 0.7 RCIM_Cubic>]}}})
      Alpha=&MakeCustomFloatFromBool(
       Bool=&SetBoolByFloatComparison(A=&VectorLength(Vector=$Engine.Owner.Velocity), B=400)
       True Float=1
       False Float=0
      )
     )
    )
    ApplyForce=false
    ApplyPosition=true
    ApplyPositionFalloff=false
   )
   &DynamicMaterialParameters(
    Index 0 Param 1=&Add_Float(
     A=$Particles.RandomFloat
     B=&FloatFromCurve(FloatCurve=DataInterface() {propertyValues {Curve {keys [
          {arriveTangent=0.1552, interpMode=RCIM_Cubic, leaveTangent=0.1552, tangentMode=RCTM_User, time=0, value=1}
          {interpMode=RCIM_Cubic, leaveTangent=0.5473, time=1, value=1.1}
         ]}}})
    )
    Index 0 Param 4=1
    Index 2 Param 1=&FloatFromCurve(
     FloatCurve=DataInterface() {propertyValues {Curve {keys [
         {interpMode=RCIM_Cubic, leaveTangent=-0.0092, tangentMode=RCTM_Break, time=0.2, value=0}
         {arriveTangent=0.8413, tangentMode=RCTM_User, time=1, value=1.3}
        ]}}}
     Scale Curve=&Lerp_Float(
      A=1
      B=1.5
      Alpha=&FloatFromCurve(FloatCurve=DataInterface() {propertyValues {Curve {keys [<at 7 0 RCIM_Cubic>,<at 7.5 1 RCIM_Cubic>]}}}, CurveIndex=$Emitter.Age)
     )
    )
    Index 2 Param 2=&FloatFromCurve(
     FloatCurve=DataInterface() {propertyValues {Curve {keys [{arriveTangent=0.7845, interpMode=RCIM_Cubic, leaveTangent=0.7845, tangentMode=RCTM_User, time=0, value=0.2}, <at 1 0.8 RCIM_Cubic>]}}}
     Scale Curve=&RandomRangeFloat(Minimum=0.2, Maximum=1.5)
    )
    Index 2 Param 3=&FloatFromCurve(
     FloatCurve=DataInterface() {propertyValues {Curve {keys [
         {interpMode=RCIM_Cubic, leaveTangent=1.0458, tangentMode=RCTM_Break, time=0, value=1}
         {interpMode=RCIM_Cubic, leaveTangent=11.1233, tangentMode=RCTM_Break, time=1, value=2}
        ]}}}
     Scale Curve=1.25
    )
    Index 2 Param 4=&FloatFromCurve(
     FloatCurve=DataInterface() {propertyValues {Curve {keys [<at 0 1 RCIM_Cubic>, {interpMode=RCIM_Cubic, leaveTangent=-1.2956, tangentMode=RCTM_Break, time=1, value=0.75}]}}}
     Scale Curve=1.5
    )
    Index 3 Param 1=&FloatFromCurve(
     FloatCurve=DataInterface() {propertyValues {Curve {keys [
         <at 0.5 0 RCIM_Cubic>
         {arriveTangent=1, interpMode=RCIM_Cubic, leaveTangent=1, time=1, value=1}
         {arriveTangent=-1, interpMode=RCIM_Cubic, leaveTangent=-1, time=1.5, value=1}
         {interpMode=RCIM_Cubic, leaveTangent=2, time=2, value=0}
        ]}}}
     CurveIndex=$Emitter.Age
     Scale Curve=&Multiply_Float(
      A=&FN_Select_Float_from_Time_Of_Day(Morning=3, Day=3, Evening=3, Night=3)
      B=&FN_FrontEndFloatModifier(Gameplay=1, Front End=2)
     )
    )
    Index 3 Param 2=&Multiply_Float(
     A=&Lerp_Float(
      A=33
      B=75
      Alpha=&FloatFromCurve(
       FloatCurve=DataInterface() {propertyValues {Curve {keys [
           <at 0.5 0 RCIM_Cubic>
           {arriveTangent=1, interpMode=RCIM_Cubic, leaveTangent=1, time=1, value=1}
           {arriveTangent=-1, interpMode=RCIM_Cubic, leaveTangent=-1, time=1.5, value=1}
           <at 2 0 RCIM_Cubic>
          ]}}}
       CurveIndex=$Emitter.Age
      )
     )
     B=&FN_FrontEndFloatModifier(Gameplay=1, Front End=1.25)
    )
    Module.Param0WriteEnabled=true
    Module.Param0WriteEnabled2=true
    Module.Param0WriteEnabled3=true
    Module.Param1WriteEnabled=false
    Module.Param1WriteEnabled2=true
    Module.Param1WriteEnabled3=true
    Module.Param2WriteEnabled=true
    Module.Param2WriteEnabled2=true
    Module.Param2WriteEnabled3=true
    Module.Param3WriteEnabled=true
    Module.Param3WriteEnabled2=true
    Module.Param3WriteEnabled3=true
    Write Parameter Index 2=true
    Write Parameter Index 3=true
   )
   &ParticleState()
   &ScaleSpriteSize(Uniform Curve Sprite Scale=DataInterface() {propertyValues {Curve {keys [
        {arriveTangent=1.6448, interpMode=RCIM_Cubic, leaveTangent=1.6448, tangentMode=RCTM_User, time=0, value=0.4}
        {arriveTangent=0.0224, interpMode=RCIM_Cubic, leaveTangent=0.0224, tangentMode=RCTM_User, time=1, value=1}
       ]}}})
   &ScaleSpriteSize(
    Scale Factor=&Multiply_Vector2D(
     A=<1 1>
     B=&MakeVector2D(
      X=1
      Y=&RemapRange(
       Input Value To Remap=&Abs_Float(Float:NiagaraFloat=&DotProduct(
         Vector A=&ReturnCameraVector(Camera Vectors=::NewEnumerator4, Coordinate Space=::Simulation, Origin Position=$Particles.Position)
         Vector B=$Engine.Owner.SystemZAxis
        ))
       Input Min=0.2
       Input Max=0.9
       Output Min=0.5
       Output Max=1
      )
     )
    )
    Scale Sprite Size Mode=::NewEnumerator2
   ) {module=ScaleSpriteSize001}
   &ScaleColor(
    Scale Alpha=&FloatFromCurve(
     FloatCurve=DataInterface() {propertyValues {Curve {keys [
         {interpMode=RCIM_Cubic, tangentMode=RCTM_User, time=0, value=0}
         {arriveTangent=0.0617, interpMode=RCIM_Cubic, leaveTangent=0.0617, tangentMode=RCTM_User, time=0.2, value=1}
         {arriveTangent=0.0658, interpMode=RCIM_Cubic, leaveTangent=0.0658, tangentMode=RCTM_User, time=0.7, value=1}
         {tangentMode=RCTM_User, time=1, value=0}
        ]}}}
     Scale Curve=&Multiply_Float(A=$User.CloudOpacity, B=&FN_FrontEndFloatModifier(Gameplay=1, Front End=0.35))
    )
    Scale RGB=&MakeVectorFromLinearColorRGB(Color=&MultiplyLinearColor(
      Linear Color=&MakeCustomLinearColorFromBool(
       Bool=$User.bIsFrontEndPreview
       True Color=<rgba 1.4 1.4 1.4 1>
       False Color=&RandomRangeLinearColor(Minimum=<rgba 1.15 1.15 1.15 1>, Maximum=<rgba 0.9368 0.9706 1.15 1>)
      )
      Scale Factor=&MakeVector4_FromLinearColor(LinearColor=&LerpLinearColors(
        StartColor=&RandomRangeLinearColor(
         Minimum=&ColorFromCurve(ColorCurve=DataInterface() {propertyValues {
            AlphaCurve {keys [<at 0 0>, {arriveTangent=1.6422, interpMode=RCIM_Cubic, leaveTangent=1.6422, time=0.1931, value=1}, <at 0.6089 1>, <at 1 0 RCIM_Cubic>]}
            BlueCurve {keys [<at 0 1 RCIM_Cubic>,<at 0.3713 0.7946>,<at 0.7985 1 RCIM_Cubic>]}
            GreenCurve {keys [<at 0 0.4991 RCIM_Cubic>,<at 0.3713 0.6458>,<at 0.7985 0.6823 RCIM_Cubic>]}
            RedCurve {keys [<at 0 0.4792 RCIM_Cubic>,<at 0.3713 1>,<at 0.7985 0.9851 RCIM_Cubic>]}
           }})
         Maximum=&ColorFromCurve(ColorCurve=DataInterface() {propertyValues {
            AlphaCurve {keys [<at 0 0>, {arriveTangent=1.6688, interpMode=RCIM_Cubic, leaveTangent=1.6688, time=0.1931, value=1}, <at 0.5992 1>, <at 1 0 RCIM_Cubic>]}
            BlueCurve {keys [<at 0 0.776>, {arriveTangent=0.224, interpMode=RCIM_Cubic, leaveTangent=0.224, time=0.444, value=1}, <at 1 1 RCIM_Cubic>]}
            GreenCurve {keys [<at 0 0.7995>, {arriveTangent=-0.2353, interpMode=RCIM_Cubic, leaveTangent=-0.2353, time=0.444, value=0.7292}, <at 1 0.5642 RCIM_Cubic>]}
            RedCurve {keys [<at 0 1>, {arriveTangent=-0.4531, interpMode=RCIM_Cubic, leaveTangent=-0.4531, time=0.444, value=0.9873}, <at 1 0.5469 RCIM_Cubic>]}
           }})
        )
        EndColor=$System.TODColor
        LerpFactor=0.05
       ))
     ))
    ScaleA=true
    ScaleRGB=true
   )
   &PointForce(
    Force Falloff Distance=200
    Force Strength=&Multiply_Float(
     A=&RandomRangeFloat(Minimum=3000, Maximum=6000)
     B=&FloatFromCurve(
      FloatCurve=DataInterface() {propertyValues {Curve {keys [<at 6 0 RCIM_Constant>, {interpMode=RCIM_Cubic, tangentMode=RCTM_User, time=7.25, value=1}, <at 7.5 0 RCIM_Cubic>]}}}
      CurveIndex=$Emitter.Age
      Scale Curve=&MaskFloatBySpawnGroup(Masked Float=1, Passthrough Float=0)
     )
    )
    Module.Use Falloff=true
    Module.UseLinearForce=false
   )
   &Drag(
    Drag=&Multiply_Float(
     A=&RandomRangeFloat(Minimum=0.7, Maximum=1.5)
     B=&FloatFromCurve(FloatCurve=DataInterface() {propertyValues {Curve {keys [<at 6.6 1 RCIM_Constant>,<at 6.9 10 RCIM_Constant>]}}}, CurveIndex=$Emitter.Age)
    )
    Use Linear Drag=true
    Use Rotational Drag=true
   )
   &SolveForcesAndVelocity(Clamp Velocity=false, Limit Acceleration=false)
  ]
  rendererClasses [&NiagaraSpriteRendererProperties]
  renderers [
   {rendererClass=&NiagaraSpriteRendererProperties}
   {rendererClass=&NiagaraSpriteRendererProperties, rendererIndex=1}
  ]
 }
 Cloud001 {
  emitterSpawnScript [
   /SetVariables(Emitter.FakeLightActive=true) {isSetParameters=true, module=SetVariables_#1}
   /SetVariables(Emitter.FrontEndShadowLow:NiagaraFloat=&FN_FrontEndFloatModifier(Gameplay=0.7, Front End=1.5)) {isSetParameters=true, module=SetVariables_#2}
  ]
  emitterUpdateScript [
   &EmitterState(Life Cycle Mode=::NewEnumerator1, Loop Behavior=::NewEnumerator1, Loop Duration=8, UseLoopDelay=false)
   &SpawnBurst_Instantaneous(Spawn Count=16, Spawn Time=1, Use Spawn Probability=false, UseLoopCountLimit=false) {module=SpawnBurst_Instantaneous002}
   &SpawnRate(
    SpawnRate=&Multiply_Float(
     A=1
     B=&FloatFromCurve(
      FloatCurve=DataInterface() {propertyValues {Curve {keys [
          <at 0 0 RCIM_Constant>
          <at 1 1 RCIM_Constant>
          {arriveTangent=-0.1333, interpMode=RCIM_Cubic, leaveTangent=-0.1333, time=4, value=1}
          {arriveTangent=-0.093, interpMode=RCIM_Cubic, leaveTangent=-0.093, tangentMode=RCTM_User, time=7, value=0.2}
         ]}}}
      CurveIndex=$Emitter.Age
     )
    )
    Use Spawn Probability=false
   )
   &SpawnBurst_Instantaneous(Spawn Count=9, Spawn Group=1, Use Spawn Probability=false, UseLoopCountLimit=false)
   &SpawnBurst_Instantaneous(Spawn Count=9, Spawn Group=2, Spawn Time=7.55, Use Spawn Probability=false, UseLoopCountLimit=false) {module=SpawnBurst_Instantaneous001}
  ]
  particleSpawnScript [
   &InitializeParticle(
    Color=<rgba 0.6 0.6 0.6 1>
    Lifetime=&MaskFloat(
     Mask Source=0
     Mask Parameter=?
     Mask=0
     Masked Float=&RandomRangeFloat(Minimum=8, Maximum=9)
     Passthrough Float=&MaskFloat(
      Mask Source=0
      Mask Parameter=?
      Mask=1
      Masked Float=&RandomRangeFloat(Minimum=3, Maximum=3)
      Passthrough Float=&RandomRangeFloat(Minimum=0.4, Maximum=0.6)
     )
    )
    Sprite Size Mode=::NewEnumerator3
    Sprite UV Mode=::NewEnumerator3
    Uniform Sprite Size=&MaskFloatBySpawnGroup(
     Masked Float=&RandomRangeFloat(Minimum=45, Maximum=60)
     Passthrough Float=&MaskFloatBySpawnGroup(
      Masked Float=&RandomRangeFloat(Minimum=55, Maximum=70)
      Passthrough Float=&RandomRangeFloat(Minimum=45, Maximum=60)
      Spawn Group Mask=1
     )
    )
    UsePositionOffset=false
    Write Material Random=true
   )
   &ShapeLocation(
    Cylinder Height=&MaskFloatBySpawnGroup(
     Masked Float=15
     Passthrough Float=&MaskFloatBySpawnGroup(Masked Float=15, Passthrough Float=0, Spawn Group Mask=1)
    )
    Cylinder Radius=&MaskFloatBySpawnGroup(
     Masked Float=50
     Passthrough Float=&MaskFloatBySpawnGroup(Masked Float=45, Passthrough Float=20, Spawn Group Mask=1)
    )
    MaskSpawnGroup=false
    Offset=&SelectVectorFromArray(
     Direct Array Index=&SpawnGroup()
     Vector Selection Array=DataInterface() {propertyValues {FloatData [<0 0 0>, <0 0 30>, <0 0 0>]}}
    )
    Offset Mode=::NewEnumerator0
    Override Seed=false
    Shape Primitive=::NewEnumerator1
    Surface Only=true
    Surface Only Band Thickness:NiagaraFloat=&SelectFloatFromArray(
     Direct Array Index=&SpawnGroup()
     Float Selection Array=DataInterface() {propertyValues {FloatData [0.75,0.8,0.5]}}
    )
    Use Endcaps In Surface Only Mode=false
    UseOwnerScale=true
   )
   &AddVelocity(
    Module.Use Falloff Distance=false
    Origin Offset=&Lerp_Vector(
     A=<0 0 35>
     B=<0 0 0>
     Alpha=&FloatFromCurve(
      FloatCurve=DataInterface() {propertyValues {Curve {defaultValue=1, keys [{tangentMode=RCTM_User, time=0, value=0}, {tangentMode=RCTM_User, time=0.5, value=1}]}}}
      CurveIndex=$Emitter.Age
     )
    )
    Override Seed=false
    Velocity Mode=::NewEnumerator1
    Velocity Speed=&Multiply_Float(
     A=&RandomRangeFloat(Minimum=20, Maximum=50)
     B=&FloatFromCurve(
      FloatCurve=DataInterface() {propertyValues {Curve {defaultValue=1, keys [{tangentMode=RCTM_User, time=0, value=1}, {tangentMode=RCTM_User, time=0.5, value=0}]}}}
      CurveIndex=$Emitter.Age
     )
    )
   ) {enabled=false}
   &CameraOffset(Camera Offset Amount=&MergeHelper_FloatSwitch(0=25, 1=45, 2=25, 3=?, Index=&SpawnGroup(), Num=::NewEnumerator1))
   &ApplyInitialForces()
   &AddVelocity(
    Velocity=<0 0 -1>
    Velocity Speed Scale=&Multiply_Float(
     A=&SelectFloatFromArray(Direct Array Index=&SpawnGroup(), Float Selection Array=DataInterface() {propertyValues {FloatData [0,45,0]}})
     B=1
    )
   ) {module=AddVelocity001}
   &AddVelocity(
    Module.Use Falloff Distance=false
    Override Seed=false
    Velocity Mode=::NewEnumerator1
    Velocity Speed=&Multiply_Float(
     A=&RandomRangeFloat(Minimum=600, Maximum=900)
     B=&FloatFromCurve(
      FloatCurve=DataInterface() {propertyValues {Curve {keys [<at 6.8 0 RCIM_Constant>,<at 7.4 1 RCIM_Constant>]}}}
      CurveIndex=$Emitter.Age
      Scale Curve=&MaskFloatBySpawnGroup(Masked Float=0, Passthrough Float=1)
     )
    )
   ) {module=AddVelocity002}
   /SetVariables(Particles.RandomFloat=&RandomRangeFloat(Minimum=0, Maximum=1)) {isSetParameters=true}
  ]
  particleUpdateScript [
   &InheritSourceMovement(
    Applied Position Scale=&Multiply_VectorByFloat(
     Vector=<1 1 1>
     Float=&Lerp_Float(
      A=&FloatFromCurve(FloatCurve=DataInterface() {propertyValues {Curve {keys [{arriveTangent=-1.0658, interpMode=RCIM_Cubic, leaveTangent=-1.0658, tangentMode=RCTM_User, time=0.1, value=1}, <at 1 0.4 RCIM_Cubic>]}}})
      B=&FloatFromCurve(FloatCurve=DataInterface() {propertyValues {Curve {keys [{arriveTangent=-0.0965, interpMode=RCIM_Cubic, leaveTangent=-0.0965, tangentMode=RCTM_User, time=0.3, value=1}, <at 1 0.7 RCIM_Cubic>]}}})
      Alpha=&MakeCustomFloatFromBool(
       Bool=&SetBoolByFloatComparison(A=&VectorLength(Vector=$Engine.Owner.Velocity), B=400)
       True Float=1
       False Float=0
      )
     )
    )
    ApplyForce=false
    ApplyPosition=true
    ApplyPositionFalloff=false
   )
   &DynamicMaterialParameters(
    Index 0 Param 1=&Add_Float(
     A=$Particles.RandomFloat
     B=&FloatFromCurve(
      FloatCurve=DataInterface() {propertyValues {Curve {keys [
          {arriveTangent=0.1552, interpMode=RCIM_Cubic, leaveTangent=0.1552, tangentMode=RCTM_User, time=0, value=1}
          {interpMode=RCIM_Cubic, leaveTangent=0.5473, time=1, value=1.1}
         ]}}}
      Scale Curve=&MergeHelper_FloatSwitch(0=5, 1=1, 2=1, 3=?, Index=&SpawnGroup(), Num=::NewEnumerator1)
     )
    )
    Index 0 Param 4=1
    Index 2 Param 1=&MaskFloatBySpawnGroup(
     Masked Float=&FloatFromCurve(FloatCurve=DataInterface() {propertyValues {Curve {keys [<at 7.6 0 RCIM_Cubic>,<at 8 1 RCIM_Cubic>]}}}, CurveIndex=$Emitter.Age)
     Passthrough Float=&FloatFromCurve(
      FloatCurve=DataInterface() {propertyValues {Curve {keys [
          {interpMode=RCIM_Cubic, leaveTangent=-0.0092, tangentMode=RCTM_Break, time=0.2, value=0}
          {arriveTangent=0.8413, tangentMode=RCTM_User, time=1, value=1.3}
         ]}}}
      Scale Curve=&Lerp_Float(
       A=1
       B=1.5
       Alpha=&FloatFromCurve(FloatCurve=DataInterface() {propertyValues {Curve {keys [<at 7 0 RCIM_Cubic>,<at 7.5 1 RCIM_Cubic>]}}}, CurveIndex=$Emitter.Age)
      )
     )
    )
    Index 2 Param 2=&FloatFromCurve(
     FloatCurve=DataInterface() {propertyValues {Curve {keys [{arriveTangent=0.7845, interpMode=RCIM_Cubic, leaveTangent=0.7845, tangentMode=RCTM_User, time=0, value=0.2}, <at 1 0.8 RCIM_Cubic>]}}}
     Scale Curve=&RandomRangeFloat(Minimum=0.2, Maximum=1.5)
    )
    Index 2 Param 3=&FloatFromCurve(
     FloatCurve=DataInterface() {propertyValues {Curve {keys [
         {interpMode=RCIM_Cubic, leaveTangent=1.0458, tangentMode=RCTM_Break, time=0, value=1}
         {interpMode=RCIM_Cubic, leaveTangent=11.1233, tangentMode=RCTM_Break, time=1, value=2}
        ]}}}
     Scale Curve=1.25
    )
    Index 2 Param 4=&FloatFromCurve(
     FloatCurve=DataInterface() {propertyValues {Curve {keys [<at 0 1 RCIM_Cubic>, {interpMode=RCIM_Cubic, leaveTangent=-1.2956, tangentMode=RCTM_Break, time=1, value=0.75}]}}}
     Scale Curve=1.5
    )
    Index 3 Param 1=&FloatFromCurve(
     FloatCurve=DataInterface() {propertyValues {Curve {keys [
         <at 0.5 0 RCIM_Cubic>
         {arriveTangent=1, interpMode=RCIM_Cubic, leaveTangent=1, time=1, value=1}
         {arriveTangent=-1, interpMode=RCIM_Cubic, leaveTangent=-1, time=1.5, value=1}
         {interpMode=RCIM_Cubic, leaveTangent=2, time=2, value=0}
        ]}}}
     CurveIndex=$Emitter.Age
     Scale Curve=&Multiply_Float(
      A=&FN_Select_Float_from_Time_Of_Day(Morning=3, Day=3, Evening=3, Night=3)
      B=&FN_FrontEndFloatModifier(Gameplay=1, Front End=2)
     )
    )
    Index 3 Param 2=&Multiply_Float(
     A=&Lerp_Float(
      A=33
      B=75
      Alpha=&FloatFromCurve(
       FloatCurve=DataInterface() {propertyValues {Curve {keys [
           <at 0.5 0 RCIM_Cubic>
           {arriveTangent=1, interpMode=RCIM_Cubic, leaveTangent=1, time=1, value=1}
           {arriveTangent=-1, interpMode=RCIM_Cubic, leaveTangent=-1, time=1.5, value=1}
           <at 2 0 RCIM_Cubic>
          ]}}}
       CurveIndex=$Emitter.Age
      )
     )
     B=&FN_FrontEndFloatModifier(Gameplay=1, Front End=1.25)
    )
    Module.Param0WriteEnabled=true
    Module.Param0WriteEnabled2=true
    Module.Param0WriteEnabled3=true
    Module.Param1WriteEnabled=false
    Module.Param1WriteEnabled2=true
    Module.Param1WriteEnabled3=true
    Module.Param2WriteEnabled=true
    Module.Param2WriteEnabled2=true
    Module.Param2WriteEnabled3=true
    Module.Param3WriteEnabled=true
    Module.Param3WriteEnabled2=true
    Module.Param3WriteEnabled3=true
    Write Parameter Index 2=true
    Write Parameter Index 3=true
   )
   &ParticleState()
   &ScaleSpriteSize(Uniform Curve Sprite Scale=DataInterface() {propertyValues {Curve {keys [
        {arriveTangent=0.5135, interpMode=RCIM_Cubic, leaveTangent=0.5135, tangentMode=RCTM_User, time=0, value=0.8}
        {arriveTangent=0.0224, interpMode=RCIM_Cubic, leaveTangent=0.0224, tangentMode=RCTM_User, time=1, value=1}
       ]}}})
   &ScaleSpriteSize(
    Scale Factor=&Multiply_Vector2D(
     A=<1 1>
     B=&MakeVector2D(
      X=1
      Y=&RemapRange(
       Input Value To Remap=&Abs_Float(Float:NiagaraFloat=&DotProduct(
         Vector A=&ReturnCameraVector(Camera Vectors=::NewEnumerator4, Coordinate Space=::Simulation, Origin Position=$Particles.Position)
         Vector B=$Engine.Owner.SystemZAxis
        ))
       Input Min=0.2
       Input Max=0.9
       Output Min=0.5
       Output Max=1
      )
     )
    )
    Scale Sprite Size Mode=::NewEnumerator2
   ) {module=ScaleSpriteSize001}
   &ScaleColor(
    Scale Alpha=&FloatFromCurve(
     FloatCurve=DataInterface() {propertyValues {Curve {keys [
         {interpMode=RCIM_Cubic, tangentMode=RCTM_User, time=0, value=0}
         {arriveTangent=0.0617, interpMode=RCIM_Cubic, leaveTangent=0.0617, tangentMode=RCTM_User, time=0.25, value=1}
         {arriveTangent=0.0658, interpMode=RCIM_Cubic, leaveTangent=0.0658, tangentMode=RCTM_User, time=0.6, value=1}
         {tangentMode=RCTM_User, time=1, value=0}
        ]}}}
     Scale Curve=&Multiply_Float(
      A=$User.CloudOpacity
      B=&FN_FrontEndFloatModifier(
       Gameplay=&MaskFloat(
        Mask Source=0
        Mask Parameter=?
        Mask=1
        Masked Float=&FloatFromCurve(
         FloatCurve=DataInterface() {propertyValues {Curve {defaultValue=1, keys [<at 0 0.35 RCIM_Cubic>, {arriveTangent=0.1477, tangentMode=RCTM_User, time=2, value=1}]}}}
         CurveIndex=$Emitter.Age
        )
        Passthrough Float=1
       )
       Front End=0.35
      )
     )
    )
    Scale RGB=&MakeVectorFromLinearColorRGB(Color=&MultiplyLinearColor(
      Linear Color=&MakeCustomLinearColorFromBool(
       Bool=$User.bIsFrontEndPreview
       True Color=<rgba 1.4 1.4 1.4 1>
       False Color=&RandomRangeLinearColor(Minimum=<rgba 1.15 1.15 1.15 1>, Maximum=<rgba 0.9368 0.9706 1.15 1>)
      )
      Scale Factor=&MakeVector4_FromLinearColor(LinearColor=&LerpLinearColors(
        StartColor=&RandomRangeLinearColor(
         Minimum=&ColorFromCurve(ColorCurve=DataInterface() {propertyValues {
            AlphaCurve {keys [<at 0 0>, {arriveTangent=1.6422, interpMode=RCIM_Cubic, leaveTangent=1.6422, time=0.1931, value=1}, <at 0.6089 1>, <at 1 0 RCIM_Cubic>]}
            BlueCurve {keys [<at 0 1 RCIM_Cubic>,<at 0.3713 0.7946>,<at 0.7985 1 RCIM_Cubic>]}
            GreenCurve {keys [<at 0 0.4991 RCIM_Cubic>,<at 0.3713 0.6458>,<at 0.7985 0.6823 RCIM_Cubic>]}
            RedCurve {keys [<at 0 0.4792 RCIM_Cubic>,<at 0.3713 1>,<at 0.7985 0.9851 RCIM_Cubic>]}
           }})
         Maximum=&ColorFromCurve(ColorCurve=DataInterface() {propertyValues {
            AlphaCurve {keys [<at 0 0>, {arriveTangent=1.6688, interpMode=RCIM_Cubic, leaveTangent=1.6688, time=0.1931, value=1}, <at 0.5992 1>, <at 1 0 RCIM_Cubic>]}
            BlueCurve {keys [<at 0 0.776>, <at 0.081 0.8169>, {arriveTangent=0.224, interpMode=RCIM_Cubic, leaveTangent=0.224, time=0.444, value=1}, <at 1 1 RCIM_Cubic>]}
            GreenCurve {keys [
              <at 0 0.7995>
              <at 0.081 0.7866>
              {arriveTangent=-0.2353, interpMode=RCIM_Cubic, leaveTangent=-0.2353, time=0.444, value=0.7292}
              <at 1 0.5642 RCIM_Cubic>
             ]}
            RedCurve {keys [
              <at 0 1>
              <at 0.081 0.9977>
              {arriveTangent=-0.4531, interpMode=RCIM_Cubic, leaveTangent=-0.4531, time=0.444, value=0.9873}
              <at 1 0.5469 RCIM_Cubic>
             ]}
           }})
        )
        EndColor=$System.TODColor
        LerpFactor=0.05
       ))
     ))
    ScaleA=true
    ScaleRGB=true
   )
   &PointForce(
    Force Falloff Distance=200
    Force Strength=&Multiply_Float(
     A=&RandomRangeFloat(Minimum=2500, Maximum=5000)
     B=&FloatFromCurve(
      FloatCurve=DataInterface() {propertyValues {Curve {keys [<at 6 0 RCIM_Constant>, {interpMode=RCIM_Cubic, tangentMode=RCTM_User, time=7.6, value=1}, <at 7.85 0 RCIM_Cubic>]}}}
      CurveIndex=$Emitter.Age
      Scale Curve=&MaskFloatBySpawnGroup(Masked Float=1, Passthrough Float=0)
     )
    )
    Module.Use Falloff=true
    Module.UseLinearForce=false
   )
   &Drag(
    Drag=&Multiply_Float(
     A=&RandomRangeFloat(Minimum=1, Maximum=1.2)
     B=&FloatFromCurve(
      FloatCurve=DataInterface() {propertyValues {Curve {keys [<at 6.6 1 RCIM_Constant>,<at 6.9 10 RCIM_Constant>]}}}
      CurveIndex=$Emitter.Age
      Scale Curve=&MergeHelper_FloatSwitch(0=1, 1=1.1, 2=1, 3=?, Index=&SpawnGroup(), Num=::NewEnumerator1)
     )
    )
    Use Linear Drag=true
    Use Rotational Drag=true
   )
   &SolveForcesAndVelocity(Clamp Velocity=false, Limit Acceleration=false)
  ]
  rendererClasses [&NiagaraSpriteRendererProperties]
  renderers [
   {rendererClass=&NiagaraSpriteRendererProperties}
   {rendererClass=&NiagaraSpriteRendererProperties, rendererIndex=1}
  ]
 }
}
summary {
 systemName=NS_Emote_SniffRust_Clouds
 emitters [
  {emitterName=Cloud001, rendererClasses [&NiagaraSpriteRendererProperties]}
  {bEnabled=false, emitterName=Cloud, rendererClasses [&NiagaraSpriteRendererProperties]}
 ]
 userVariables [
  User.bIsFrontEnd {type=NiagaraBool}
  User.bIsFrontEndPreview {type=NiagaraBool}
  User.CloudOpacity {type=NiagaraFloat, defaultValue=1, description=""}
 ]
}
